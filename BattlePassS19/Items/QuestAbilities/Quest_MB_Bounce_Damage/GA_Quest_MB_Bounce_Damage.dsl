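; Quest gameplay ability that tracks damage the player deals during a bounce, from launch until landing.
; Activation from an event stores the instigator's tags, then waits for the tracked landing tag while listening in parallel for repeated Event.Damage.Dealt events.
; Each damage event's truncated magnitude goes to the player's quest manager as a complex custom stat event.
; Landing removes configured loose tags from the player pawn and ends the ability.
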(event Ability|EventActivateAbilityFromEvent
  (params (EventData :type Gameplay%20Event%20Data%20Structure%20%28by%20ref%29))
  (bind (_eventtag _instigator _target _optionalobject _optionalobject2 _contexthandle _instigatortags) (Utilities|Struct|BreakGameplayEventData EventData))
  (|CallParentFunction EventData)
  (Variables|Tags|SetPlayerTags _instigatortags)
  (Development|PrintString "@@@ ActivateAbilityFromEvent []: CALLED -  Launch!" false (Variables|Debug|GetbDebugAbility))
  ;[1 0.2402 0.03963] Initialize bounce tracking...
  (bind _asynctaskproxy (Ability|Tasks|WaitGameplayEvent (Variables|Tags|GetTrackedLandingTag) 0 true)
    (:then
      (Variables|Default|SetEventTask _asynctaskproxy)
      (bind (_asynctaskproxy_2 _payload) (Ability|Tasks|WaitGameplayEvent "(TagName=\"Event.Damage.Dealt\")" 0 false false)
        (:EventReceived
          (bind (_payload_eventtag _payload_instigator _payload_target _payload_optionalobject _payload_optionalobject2 _payload_contexthandle _payload_instigatortags _payload_targettags _payload_eventmagnitude) (Utilities|Struct|BreakGameplayEventData _payload))
          (Quest|SendComplexCustomStatEvent (Variables|GAVars|Data|GetOwningPlayerQuestManager) 0 (GameplayTags|MakeGameplayTagContainerfromTag (Variables|Default|GetStatEvent)) "(GameplayTags=)" (Math|Float|Truncate _payload_eventmagnitude)))))
    (:EventReceived
      ;[1 0.09399 0.7665] Target landed...
      (Ability|GameplayEffect|RemoveLooseGameplayTags (Variables|GAVars|Data|GetOwningPlayerPawn) (Variables|Default|GetTagsToRemove))
      (Ability|EndAbility))))
